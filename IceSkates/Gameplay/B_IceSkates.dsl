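; B_IceSkates is the Fortnite weapon Blueprint for the ice skates item, marking the owner while the skates are selected.
; On equip, it adds a selected tag to the owner, sets HUD key-action visibility and plays the reload-end library sound on itself.
; On local unequip, it removes the tag and, if a ScalableFloat toggle allows, the equipped GameplayEffect.
; Nothing here applies that effect or implements skating movement.
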(event EventBeginPlay
  (|CallParentFunction) ; This node is disabled and will not be called.\nDrag off pins to build functionality.
)

(event EventTick
  (params (DeltaSeconds :type Float%20%28single-precision%29))
  (|CallParentFunction DeltaSeconds) ; This node is disabled and will not be called.\nDrag off pins to build functionality.
)

(event FortWeapon|EventOnLocalUnEquip
  (bind _returnvalue (Actor|GetOwner))
  (Ability|GameplayEffect|RemoveLooseGameplayTags _returnvalue (Variables|Default|GetTag_IceSkatesSelected))
  (if (ScalableFloat|GetValueasBool (Variables|Default|GetRow_ShouldDisableOnUnequip))
    (GameplayEffects|RemoveActiveGameplayEffectbySourceEffect (Ability|GetAbilitySystemComponent _returnvalue) (Variables|Default|GetGE_IceSkatesEquipped) 0 -1)))

(event FortWeapon|EventOnEquippedAndAttached
  (bind _self self)
  (Ability|GameplayEffect|AddLooseGameplayTags (Actor|GetOwner) (Variables|Default|GetTag_IceSkatesSelected))
  (WeaponHUDData|SetHudKeyActionVisibility 0 (ScalableFloat|GetValueasBool (Variables|Default|GetRow_ShouldShowWeaponHudMessage)))
  (Utilities|IsValid _self
    (:Is%20Valid
      (Utilities|PlaySoundLibraryAttached "(TagName=\"SoundLibrary.Weapon.Reload.End\")" _self))))

(fn ConstructionScript
  (|CallParentFunction))
